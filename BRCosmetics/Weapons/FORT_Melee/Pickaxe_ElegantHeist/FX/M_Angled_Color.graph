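# Translucent unlit Niagara mesh particle material for the ElegantHeist pickaxe FX, giving meshes an angle-dependent color with a moving-highlight look.
# The absolute normal-to-camera dot product samples a curve atlas, then gets a dynamic-parameter power and particle color tint.
# Rotated screen UVs, scaled by Sheen Scale and offset by side-facing normals, form a sine-sharpened sheen band.
# That band blends the color toward white before MF_FX_EmissiveAdjustments.
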Material /BRCosmetics/Weapons/FORT_Melee/Pickaxe_ElegantHeist/FX/M_Angled_Color
graph {
 properties {
  bAutomaticallySetUsageInEditor=false
  bUsedWithNiagaraMeshParticles=true
  blendMode=BLEND_Translucent
  shadingModel=MSM_Unlit
  usedShadingModels=Unlit
 }
 nodes [
  Constant()
  n1 = VertexNormalWS()
  n2 = Multiply(A=MaterialFunctionCall() {materialFunction=&CameraDirectionVector}.Result) {constB=-1}
  n3 = CurveAtlasRowParameter(CurveTime=Abs(None=DotProduct(A=@n1, B=@n2))) {
   atlas=&/CA_AngleColors
   curve=&/CC_AngleColor_1
   parameterName=ColorCurve
  }
  n4 = Power(Base=@n3, Exp=DynamicParameter() {paramNames [Power, Sheen Scale, Param3, Param4]}.Power)
  n5 = ParticleColor()
  n6 = Subtract(A=Rotator(Coordinate=ScreenPosition().ViewportUV, Time=Constant() {r=0.6})) {constB=0.5}
  n7 = Add(A=Multiply(A=Multiply(A=@n6) {constB=10}, B=DynamicParameter() {paramNames [Power, Sheen Scale, Param3, Param4]}.Sheen Scale)) {constB=0.5}
  n8 = MaterialFunctionCall() {materialFunction=&CameraDirectionVector}
  n9 = Constant3Vector() {constant=<rgba 0 0 1 0>}
  n10 = DotProduct(A=VertexNormalWS(), B=Multiply(A=CrossProduct(A=@n8.Result, B=@n9)) {constB=-1})
  n11 = Saturate(None=Add(A=ComponentMask(None=@n7) {r=true}, B=Multiply(A=@n10) {constB=3}))
  n12 = SmoothStep(Value=Power(Base=Sine(None=@n11)) {constExponent=5}) {constMax=0.2, constMin=0.1, constValue=0.6}
  MaterialFunctionCall(Emissive=LinearInterpolate(A=Multiply(A=@n4, B=@n5.RGB), Alpha=Saturate(None=@n12)) {constAlpha=0}) {materialFunction=&MF_FX_EmissiveAdjustments}
  TextureCoordinate()
 ]
}
