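# Niagara swing effect for the ElegantHeist pickaxe cosmetic, combining three ribbon trails with glinting particles.
# ChromaTrail, Emissive_Filler_Trail and Trail_Details use shared anim-trail modules, each with its own color-over-life gradient and width curve.
# Bling_Big spawns three distance-based groups at socket positions, using a VisibilityTag to switch per-group looks across sprite and mesh renderers, with flicker and bokeh parameters.
# Time-of-day values tint or dim the trails and particles.
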NiagaraSystem /BRCosmetics/Weapons/FORT_Melee/Pickaxe_ElegantHeist/FX/NS_Pickaxe_ElegantHeist_Swing
dependencies {
 usedDataInterfaces [
  &NiagaraDataInterfaceCurve
  &NiagaraDataInterfaceColorCurve
  &NiagaraDataInterfaceCamera
  &NiagaraDataInterfaceVector2DCurve
  &NiagaraDataInterfaceSocketReader
 ]
 usedDynamicInputs [&ReturnCameraVector]
 usedModules [
  &self:SystemSpawnScript.SystemScriptSource.SystemScriptGraph.NiagaraNodeAssignment_3.SetVariables_#1
  &self:SystemSpawnScript.SystemScriptSource.SystemScriptGraph.NiagaraNodeAssignment_0.SetVariables_#2
  &SystemState
 ]
}
emitters {
 Bling_Big {
  emitterUpdateScript [
   &EmitterState(Life Cycle Mode=::NewEnumerator1, Loop Behavior=::NewEnumerator1, Loop Delay=0.1, Loop Duration=1, UseLoopDelay=true)
   &SpawnPerUnit(Module.Use Max Movement Threshold=true, Module.Use Movement Tolerance=false, Spawn Spacing=30, Use Spawn Probability=true)
   &SpawnPerUnit(Module.Use Max Movement Threshold=true, Module.Use Movement Tolerance=false, Spawn Group=1, Spawn Spacing=60, Use Spawn Probability=false) {module=SpawnPerUnit001}
   &SpawnPerUnit(
    Module.Use Max Movement Threshold=true
    Module.Use Movement Tolerance=false
    Spawn Group=2
    Spawn Probability=1
    Spawn Spacing=10
    Use Spawn Probability=false
   ) {module=SpawnPerUnit002}
  ]
  particleSpawnScript [
   /SetVariables(Particles.VisibilityTag=&MergeHelper_IntSwitch(0=0, 1=1, 2=2, 3=?, 4=?, 5=?, Index=$Emitter.SpawnGroup, Num=::NewEnumerator1)) {isSetParameters=true}
   /SetVariables(
    Particles.CustomRandom:NiagaraFloat=&RandomRangeFloat(Minimum=0.2, Maximum=0.8)
    Particles.CustomRandomTriangle:NiagaraFloat=&RandomRangeFloat(Minimum=0, Maximum=0.3)
   ) {isSetParameters=true}
   &InitializeParticle(
    Color=&MultiplyLinearColor(
     Linear Color=&MergeHelper_ColorSwitch(
      0=&RandomRangeLinearColor(Minimum=<rgba 2 1.725 0.7396 1>, Maximum=<rgba 2 0.9554 0 1>)
      1=<rgba 0 0 0 0>
      2=&Multiply_LinearColorByFloat(Linear Color=<rgba 1 1 1 1>, Float=&Lerp_Float(A=1, B=2, Alpha=$Particles.CustomRandomTriangle))
      3=?
      Index=$Particles.VisibilityTag
      Num=::NewEnumerator1
     )
     Scale Factor=&MakeVector4_FromLinearColor(LinearColor=$System.TOD)
    )
    Lifetime=&MergeHelper_FloatSwitch(
     0=&RandomRangeFloat(Minimum=0.4, Maximum=0.6)
     1=&RandomRangeFloat(Minimum=0.4, Maximum=0.65)
     2=&FloatFromCurve(
      FloatCurve=DataInterface() {propertyValues {Curve {keys [<at 0 0.2>,<at 0.9 0.6>,<at 1 1>]}}}
      CurveIndex=&RandomRangeFloat(Minimum=0, Maximum=1)
     )
     3=&RandomRangeFloat(Minimum=0.1, Maximum=0.1)
     Index=$Emitter.SpawnGroup
     Num=::NewEnumerator1
    )
    Mesh Scale=&Multiply_VectorByFloat(
     Vector=&MakeVector(
      X=&RandomRangeFloat(Minimum=1, Maximum=1)
      Y=&RandomRangeFloat(Minimum=0.2, Maximum=2)
      Z=&RandomRangeFloat(Minimum=1, Maximum=1)
     )
     Float=&Multiply_Float(
      A=&RandomRangeFloat(Minimum=0.3, Maximum=1)
      B=&FloatFromCurve(
       FloatCurve=DataInterface() {propertyValues {Curve {keys [<at 0 0.3>,<at 0.8 0.4>,<at 1 1>]}}}
       CurveIndex=&RandomRangeFloat(Minimum=0, Maximum=1)
       Scale Curve=0.23
      )
     )
    )
    Mesh Scale Mode=::NewEnumerator4
    Sprite Rotation Angle=&MergeHelper_FloatSwitch(
     0=0
     1=0
     2=&RandomRangeFloat(Minimum=0, Maximum=360)
     3=?
     Index=$Particles.VisibilityTag
     Num=::NewEnumerator1
    )
    Sprite Rotation Mode=::NewEnumerator0
    Sprite Size=&MergeHelper_Vector2Switch(
     0=&RandomRangeVector2D(Minimum=<30 15>, Maximum=<40 30>)
     1=<0 0>
     2=&RandomRangeVector2D(Minimum=<15 7>, Maximum=<15 10>)
     3=?
     4=?
     Index=$Particles.VisibilityTag
     Num=::NewEnumerator1
    )
    Sprite Size Mode=::NewEnumerator4
    UsePositionOffset=false
    Write Material Random=true
   )
   &SocketLocation(
    Interpolate Socket Sampling=true
    SamplePosOffset=true
    Sampled Position Offset=&MergeHelper_VectorSwitch(
     0=&RandomRangeVector(Minimum=<-3 -3 -3>, Maximum=<3 3 3>)
     1=<0 0 0>
     2=<0 0 -10>
     3=?
     Index=$Particles.VisibilityTag
     Num=::NewEnumerator1
    )
    Socket Index=&MergeHelper_IntSwitch(
     0=&RandomRangeInt(Minimum=0, Maximum=1)
     1=&RandomRangeInt(Minimum=0, Maximum=1)
     2=0
     3=?
     4=?
     5=?
     Index=$Particles.VisibilityTag
     Num=::NewEnumerator1
    )
    Socket Reader=$System.Socket Reader
    Socket Sampling Mode=::NewEnumerator3
    VelocitySpeedLimit=false
   )
   /SetVariables(Particles.InitialSize:Vector2f=&Vector2DFromFloat(Value=&RandomRangeFloat(Minimum=20, Maximum=30))) {isSetParameters=true}
   &ShapeLocation(
    Apply To Particle Position=::NewEnumerator1
    MaskSpawnGroup=false
    Override Seed=false
    Sphere Radius=&Multiply_Float(A=0, B=1)
    UseOwnerScale=true
   ) {module=ShapeLocation001}
   /SetVariables(
    Particles.Bokeh Color:LinearColor=&RandomRangeLinearColor(Minimum=<rgba 0.6771 0.6771 0.6771 1>, Maximum=<rgba 1 1 1 1>)
    Particles.Bokeh Position:NiagaraPosition=$Particles.ShapeLocation001.ShapePosition
    Particles.Bokeh Size:Vector2f=&Vector2DFromFloat(Value=&Multiply_Float(A=&RandomRangeFloat(Minimum=15, Maximum=25), B=1))
    Particles.Bokeh Velocity:Vector3f=&Multiply_VectorByFloat(Vector=<1 1 1>, Float=&RandomRangeFloat(Minimum=0.2, Maximum=0.5))
    Particles.DynamicBokeh:Vector4f=<10 1 0 0>
   ) {isSetParameters=true}
   &InitialMeshOrientation(Use Rotation Vector=false)
   &AddVelocity(
    Velocity=&RandomRangeVector(Minimum=<-1 -1 -1>, Maximum=<1 1 1>)
    Velocity Speed Scale=&RandomRangeFloat(Minimum=5, Maximum=50)
   )
   /SetVariables(Particles.DynamicTriangles:Vector4f=&MakeVector4(X=&Lerp_Float(A=1, B=2, Alpha=$Particles.CustomRandomTriangle), Y=3, Z=1, W=1)) {isSetParameters=true}
  ]
  particleUpdateScript [
   &Drag(
    Drag=&RandomRangeFloat(Minimum=3, Maximum=4)
    Rotational Drag=0
    Use Linear Drag=true
    Use Rotational Drag=true
   )
   &SolveForcesAndVelocity(Clamp Velocity=false, Limit Acceleration=false)
   &ParticleState()
   &ScaleMeshSize(Scale Factor=&VectorFromFloat(Value=&FloatFromCurve(FloatCurve=DataInterface() {propertyValues {Curve {keys [
          {arriveTangent=14.45, interpMode=RCIM_Cubic, time=0, value=0}
          {arriveTangent=2.5, leaveTangent=2.5, time=0.1, value=1}
          <at 0.5 1>
          <at 1 0>
         ]}}})))
   &UpdateMeshOrientation(
    Rotation Rate=&RandomRangeFloat(Minimum=0.5, Maximum=1)
    Rotation Vector=&RandomRangeVector(Minimum=<0 0 0>, Maximum=<0 1 1>)
   )
   /SetVariables(Particles.TimeOfDay:NiagaraFloat=&FNTimeOfDayScalar(Night Scale=0.5, Morning Scale=0.3, Evening Scale=0.3)) {isSetParameters=true}
   /SetVariables(Particles.Color2=&MultiplyLinearColor(
     Linear Color=&ColorFromCurve(ColorCurve=DataInterface() {propertyValues {
        AlphaCurve {keys [<at 0 0>,<at 0.15 0.2 RCIM_Cubic>,<at 1 0 RCIM_Cubic>]}
        BlueCurve {keys [<at 0.1483 0.1562 RCIM_Cubic>,<at 1 0.0156 RCIM_Cubic>]}
        GreenCurve {keys [<at 0.1483 0.7188 RCIM_Cubic>,<at 1 0.5898 RCIM_Cubic>]}
        RedCurve {keys [<at 0.1483 1 RCIM_Cubic>,<at 1 1 RCIM_Cubic>]}
       }})
     Scale Factor=&MakeFloat4FromVectorAndFloat("Vector (RGB)"=<1 1 1>, "Float (Alpha)"=$Particles.TimeOfDay)
    )) {isSetParameters=true}
   &ScaleSpriteSize(Scale Factor=&MergeHelper_Vector2Switch(
     0=&Vector2DFromFloat(Value=&Multiply_Float(
       A=&FloatFromCurve(FloatCurve=DataInterface() {propertyValues {Curve {keys [<at 0.5 1>, {tangentMode=RCTM_User, time=1, value=0}]}}})
       B=&Lerp_Float(
        A=1
        B=&FloatFromCurve(
         FloatCurve=DataInterface() {propertyValues {Curve {keys [
             <at 0 0 RCIM_Cubic>
             {arriveTangent=-0.0175, interpMode=RCIM_Cubic, leaveTangent=-0.0175, tangentMode=RCTM_User, time=0.1563, value=0.907}
             {arriveTangent=-0.0042, interpMode=RCIM_Cubic, leaveTangent=-0.0042, tangentMode=RCTM_User, time=0.2646, value=0.0384}
             {arriveTangent=0.0081, interpMode=RCIM_Cubic, leaveTangent=0.0081, tangentMode=RCTM_User, time=0.4692, value=0.7123}
             {arriveTangent=0.034, interpMode=RCIM_Cubic, leaveTangent=0.034, tangentMode=RCTM_User, time=0.6406, value=0.0348}
             {arriveTangent=0.019, interpMode=RCIM_Cubic, leaveTangent=0.019, tangentMode=RCTM_User, time=0.7993, value=0.5053}
             <at 1 0 RCIM_Cubic>
            ]}}}
         CurveIndex=&Multiply_Float(A=$Particles.NormalizedAge, B=$Particles.CustomRandom)
        )
        Alpha=0.5
       )
      ))
     1=<0 0>
     2=&Vector2DFromCurve(Vector2Curve=DataInterface() {propertyValues {XCurve {keys [<at 0 1>,<at 1 0>]}, YCurve {keys [<at 0 1>,<at 1 0>]}}}, ScaleCurve=<1 1>)
     3=?
     4=?
     Index=$Particles.VisibilityTag
     Num=::NewEnumerator1
    )) {module=ScaleSpriteSize001}
   /SetVariables(Particles.SpriteSize2=&Multiply_Vector2(
     A=$Particles.InitialSize
     B=&Vector2DFromFloat(Value=&FloatFromCurve(
       FloatCurve=DataInterface() {propertyValues {Curve {keys [
           <at 0 0 RCIM_Cubic>
           {arriveTangent=-0.0175, interpMode=RCIM_Cubic, leaveTangent=-0.0175, tangentMode=RCTM_User, time=0.1563, value=0.907}
           {arriveTangent=-0.0042, interpMode=RCIM_Cubic, leaveTangent=-0.0042, tangentMode=RCTM_User, time=0.2646, value=0.0384}
           {arriveTangent=0.0081, interpMode=RCIM_Cubic, leaveTangent=0.0081, tangentMode=RCTM_User, time=0.4692, value=0.7123}
           {arriveTangent=0.034, interpMode=RCIM_Cubic, leaveTangent=0.034, tangentMode=RCTM_User, time=0.6406, value=0.0348}
           {arriveTangent=0.019, interpMode=RCIM_Cubic, leaveTangent=0.019, tangentMode=RCTM_User, time=0.7993, value=0.5053}
           <at 1 0 RCIM_Cubic>
          ]}}}
       CurveIndex=&Multiply_Float(A=$Particles.NormalizedAge, B=$Particles.CustomRandom)
      ))
    )) {isSetParameters=true}
   &ScaleColor(
    Scale Alpha=&MergeHelper_FloatSwitch(
     0=&FloatFromCurve(FloatCurve=DataInterface() {propertyValues {Curve {keys [<at 0 0>,<at 0.1 1>,<at 0.5 1>,<at 1 0>]}}})
     1=0
     2=&Multiply_Float(
      A=0.6
      B=&FloatFromCurve(FloatCurve=DataInterface() {propertyValues {Curve {keys [<at 0.5 1>,<at 1 0.5>]}}})
     )
     3=?
     Index=$Particles.VisibilityTag
     Num=::NewEnumerator1
    )
    Scale RGB=&MergeHelper_VectorSwitch(
     0=<1 1 1>
     1=<1 1 1>
     2=&MakeVectorFromLinearColorRGB(Color=<rgba 1 1 1 1>)
     3=?
     Index=$Particles.VisibilityTag
     Num=::NewEnumerator1
    )
    ScaleA=true
    ScaleRGB=true
   )
   /SetVariables(
    Particles.Bokeh Color:LinearColor=&MultiplyLinearColor(
     Linear Color=&MakeLinearColorFromVectorAndFloat(
      "Vector (RGB)"=&MakeVectorFromLinearColorRGB(Color=&ColorFromCurve(
        ColorCurve=DataInterface() {propertyValues {
          AlphaCurve {keys [<at 0 1 RCIM_Cubic>]}
          BlueCurve {keys [<at 0 1>, <at 0.2876 0.5478 RCIM_Cubic>, {arriveTangent=0.6348, interpMode=RCIM_Cubic, leaveTangent=0.6348, time=0.6772, value=1}, <at 1 1>]}
          GreenCurve {keys [
            <at 0 0.7043>
            {arriveTangent=-0.2096, interpMode=RCIM_Cubic, time=0.2876, value=0.9616}
            {arriveTangent=-0.3612, interpMode=RCIM_Cubic, leaveTangent=-0.3612, time=0.6772, value=0.5677}
            <at 1 0.7043>
           ]}
          RedCurve {keys [
            <at 0 0.6198>
            {arriveTangent=0.4224, interpMode=RCIM_Cubic, time=0.2876, value=1}
            {arriveTangent=-0.5337, interpMode=RCIM_Cubic, leaveTangent=-0.5337, time=0.6772, value=0.895}
            <at 1 0.6198>
           ]}
         }}
        CurveIndex=&Frac_Float(Float=&Add_Float(
          A=&Multiply_Float(A=$Particles.NormalizedAge, B=0.3)
          B=&RandomRangeFloat(Minimum=0, Maximum=1)
         ))
       ))
      "Float (Alpha)"=&FloatFromCurve(
       FloatCurve=DataInterface() {propertyValues {Curve {keys [{interpMode=RCIM_Cubic, leaveTangent=-2.4411, tangentMode=RCTM_Break, time=0.3, value=1}, <at 1 0>]}}}
       Scale Curve=&RandomRangeFloat(Minimum=0.02, Maximum=0.06)
      )
     )
     Scale Factor=&MakeVector4_FromLinearColor(LinearColor=$Particles.Initial.Bokeh Color)
    )
    Particles.Bokeh Position:NiagaraPosition=&AddVectorToPosition(Position=$Particles.Bokeh Position, Vector=<0 0 0>)
    Particles.Bokeh Size:Vector2f=&Multiply_Vector2D_ByFloat(
     Float=&FloatFromCurve(FloatCurve=DataInterface() {propertyValues {Curve {keys [
          {arriveTangent=5.0653, interpMode=RCIM_Cubic, leaveTangent=5.0653, tangentMode=RCTM_User, time=0, value=0}
          {arriveTangent=1, interpMode=RCIM_Cubic, leaveTangent=1, time=0.3, value=0.8}
          <at 1 1 RCIM_Cubic>
         ]}}})
     Vector2D=$Particles.Initial.Bokeh Size
    )
   ) {isSetParameters=true}
  ]
  rendererClasses [&NiagaraSpriteRendererProperties, &NiagaraMeshRendererProperties]
  renderers [
   {rendererClass=&NiagaraSpriteRendererProperties}
   {rendererClass=&NiagaraSpriteRendererProperties, rendererIndex=1}
   {rendererClass=&NiagaraSpriteRendererProperties, rendererIndex=2}
   {rendererClass=&NiagaraMeshRendererProperties, rendererIndex=3}
   {rendererClass=&NiagaraSpriteRendererProperties, rendererIndex=4}
   {rendererClass=&NiagaraMeshRendererProperties, rendererIndex=5}
  ]
 }
 ChromaTrail {
  emitterUpdateScript [
   &EmitterState()
   &FN_SpawnAnimTrail()
  ]
  particleSpawnScript [
   &FN_InitializeAnimTrailBoneIndecies()
   &FN_InitializeAnimTrail(Lifetime=0.3)
  ]
  particleUpdateScript [
   &FN_UpdateAnimTrail() {module=FN_UpdateAnimTrail001}
   &Color(
    Color=&ColorFromCurve(ColorCurve=DataInterface() {propertyValues {
       AlphaCurve {keys [<at 0 1 RCIM_Cubic>,<at 0.9978 0 RCIM_Cubic>]}
       BlueCurve {keys [<at 0 1 RCIM_Cubic>,<at 0.3173 1>,<at 0.8564 1>]}
       GreenCurve {keys [<at 0 1 RCIM_Cubic>,<at 0.3173 0.8837>,<at 0.8564 0.3306>]}
       RedCurve {keys [<at 0 1 RCIM_Cubic>,<at 0.3173 0.592>,<at 0.8564 0.511>]}
      }})
    Scale Alpha=0.3
   )
   &ScaleColor(
    Color Value To Scale=$Particles.Color
    Scale Mode=::NewEnumerator1
    Scale RGBA=&MakeVector4_FromLinearColor(LinearColor=$System.TOD)
    ScaleRGBA=true
   )
   &SolveForcesAndVelocity(Clamp Velocity=false, Limit Acceleration=false)
   &ScaleRibbonWidth(Ribbon Width Scale=&FloatFromCurve(FloatCurve=DataInterface() {propertyValues {Curve {keys [{arriveTangent=1.258, interpMode=RCIM_Cubic, leaveTangent=1.258, tangentMode=RCTM_User, time=0, value=0.8}, <at 1 1.3 RCIM_Cubic>]}}}))
   &DynamicMaterialParameters(
    Module.Index0Float0Write=true
    Module.Index0Float1Write=true
    Module.Index0Float2Write=true
    Module.Index0Float3Write=true
    Module.Index1Float0Write=true
    Module.Index1Float1Write=true
    Module.Index1Float2Write=true
    Module.Index1Float3Write=true
    Write Parameter Index 1=::NewEnumerator1
   )
   /SetVariables(Particles.InvAge:NiagaraFloat=&OneMinusFloat(Float=$Particles.NormalizedAge)) {isSetParameters=true}
  ]
  rendererClasses [&NiagaraRibbonRendererProperties]
  renderers [{rendererClass=&NiagaraRibbonRendererProperties}]
 }
 Emissive_Filler_Trail {
  emitterUpdateScript [
   &EmitterState(Life Cycle Mode=::NewEnumerator1, Loop Behavior=::NewEnumerator1, Loop Duration=0.27, UseLoopDelay=true)
   &FN_SpawnAnimTrail()
  ]
  particleSpawnScript [
   &FN_InitializeAnimTrailBoneIndecies(First Bone or Socket Index=1, Second Bone or Socket Index=0)
   &FN_InitializeAnimTrail(
    Color=&MultiplyLinearColor(Linear Color=<rgba 1 1 1 2>, Scale Factor=&MakeVector4_FromLinearColor(LinearColor=$System.TOD))
    Lifetime=0.3
   ) {module=FN_InitializeAnimTrail001}
   &SystemLocation(MaskSpawnGroup=false)
  ]
  particleUpdateScript [
   &FN_UpdateAnimTrail()
   &ScaleRibbonWidth(Ribbon Width Scale=&FloatFromCurve(FloatCurve=DataInterface() {propertyValues {Curve {keys [{interpMode=RCIM_Cubic, tangentMode=RCTM_User, time=0.2, value=1}, {interpMode=RCIM_Cubic, tangentMode=RCTM_User, time=1, value=0}]}}}))
   &ScaleColor(
    Scale Alpha=&Multiply_Float(
     A=&FloatFromCurve(FloatCurve=DataInterface() {propertyValues {Curve {keys [<at 0.3 1 RCIM_Cubic>,<at 1 0 RCIM_Cubic>]}}}, Scale Curve=1.5)
     B=&FNTimeOfDayScalar(Night Scale=1, Morning Scale=0.5, Evening Scale=0.5)
    )
    Scale RGB=&MakeVectorFromLinearColorRGB(Color=&ColorFromCurve(ColorCurve=DataInterface() {propertyValues {
        AlphaCurve {keys [<at 0 1 RCIM_Cubic>,<at 1 0 RCIM_Cubic>]}
        BlueCurve {keys [<at 0 0.3315>, {arriveTangent=1.3882, interpMode=RCIM_Cubic, leaveTangent=1.3882, time=0.1232, value=1}, <at 0.399 0.8854>]}
        GreenCurve {keys [<at 0 0.7175>, {arriveTangent=-1.366, interpMode=RCIM_Cubic, leaveTangent=-1.366, time=0.1232, value=0.4864}, <at 0.399 0.1724>]}
        RedCurve {keys [<at 0 1>, {arriveTangent=-1.9244, interpMode=RCIM_Cubic, leaveTangent=-1.9244, time=0.1232, value=0.1457}, <at 0.399 0.2321>]}
       }}))
    ScaleA=true
    ScaleRGB=true
   ) {module=ScaleColor002}
  ]
  rendererClasses [&NiagaraRibbonRendererProperties]
  renderers [{rendererClass=&NiagaraRibbonRendererProperties}]
 }
 Trail_Details {
  emitterSpawnScript [/SetVariables(Emitter.Trail Random:NiagaraFloat=&RandomRangeFloat(Minimum=0, Maximum=100)) {isSetParameters=true}]
  emitterUpdateScript [
   &EmitterState()
   &FN_SpawnAnimTrail()
  ]
  particleSpawnScript [
   &FN_InitializeAnimTrailBoneIndecies(First Bone or Socket Index=1, Second Bone or Socket Index=0) {module=FN_InitializeAnimTrailBoneIndecies001}
   &FN_InitializeAnimTrail(
    AnimTrailWidth=&FloatFromCurve(
     FloatCurve=DataInterface() {propertyValues {Curve {keys [<at 0.02 0>,<at 0.07 1>]}}}
     CurveIndex=$System.Age
     Scale Curve=&MergeHelper_FloatSwitch(0=1, 1=2, 2=?, 3=?, Index=&SpawnGroup(), Num=::NewEnumerator0)
    )
    Color=&MultiplyLinearColor(Linear Color=<rgba 1 1 1 0.4>, Scale Factor=&MakeVector4_FromLinearColor(LinearColor=$System.TOD))
   ) {module=FN_InitializeAnimTrail001}
   &DynamicMaterialParameters(
    Index 0 Param 1=$Emitter.Trail Random
    Module.Index0Float0Write=true
    Module.Index0Float1Write=false
    Module.Index0Float2Write=false
    Module.Index0Float3Write=false
   )
  ]
  particleUpdateScript [
   &Drag(Drag=&RandomRangeFloat(Minimum=5, Maximum=8), Use Linear Drag=true, Use Rotational Drag=true)
   &SolveForcesAndVelocity(Clamp Velocity=false, Limit Acceleration=false)
   &ScaleColor(
    Scale Alpha=&FloatFromCurve(FloatCurve=DataInterface() {propertyValues {Curve {keys [{arriveTangent=-2.2741, interpMode=RCIM_Cubic, time=0, value=0.3}, <at 1 0 RCIM_Cubic>]}}})
    Scale RGB=&Multiply_Vector(
     A=&MakeVectorFromLinearColorRGB(Color=&ColorFromCurve(ColorCurve=DataInterface() {propertyValues {
         AlphaCurve {keys []}
         BlueCurve {keys [<at 0.0168 1>,<at 0.5867 0.3229>,<at 0.9987 0.1415>]}
         GreenCurve {keys [<at 0.0168 0.9347>,<at 0.5867 0.7519>,<at 0.9987 0.2158>]}
         RedCurve {keys [<at 0.0168 0.9219>,<at 0.5867 1>,<at 0.9987 0.5156>]}
        }}))
     B=&MakeVectorFromLinearColorRGB(Color=&ColorFromCurve(
       ColorCurve=DataInterface() {propertyValues {
         AlphaCurve {keys [<at 0 1>,<at 0.125 1>,<at 0.25 1>,<at 0.375 1>,<at 0.5 1>,<at 0.625 1>,<at 0.75 1>,<at 0.875 1>,<at 1 1>]}
         BlueCurve {keys [<at 0.0566 0.9216>,<at 0.166 0.6172>,<at 0.208 0.5395>,<at 0.25 0.6172>,<at 0.291 0.6939>,<at 0.333 0.7379>,<at 0.375 0.7835>,<at 0.416 0.7991>,<at 0.458 0.8148>,<at 0.5 0.8308>,<at 0.541 0.8148>,<at 0.583 0.7991>,<at 0.625 0.7991>,<at 0.666 0.7529>,<at 0.708 0.7084>,<at 0.75 0.5711>,<at 0.791 0.4735>,<at 0.9596 0.9216>]}
         GreenCurve {keys [<at 0.0566 0.6975>,<at 0.166 0.5333>,<at 0.208 0.5271>,<at 0.25 0.6376>,<at 0.291 0.7529>,<at 0.333 0.7835>,<at 0.375 0.8148>,<at 0.416 0.7991>,<at 0.458 0.8148>,<at 0.5 0.8469>,<at 0.541 0.7991>,<at 0.583 0.7991>,<at 0.625 0.8148>,<at 0.666 0.7991>,<at 0.708 0.7529>,<at 0.75 0.5711>,<at 0.791 0.5029>,<at 0.9596 0.6341>]}
         RedCurve {keys [<at 0.0566 0.4464>,<at 0.166 0.7084>,<at 0.208 0.6867>,<at 0.25 0.8148>,<at 0.291 0.8879>,<at 0.333 0.8963>,<at 0.375 0.9131>,<at 0.416 0.8714>,<at 0.458 0.8714>,<at 0.5 0.9131>,<at 0.541 0.8632>,<at 0.583 0.8714>,<at 0.625 0.9131>,<at 0.666 0.8963>,<at 0.708 0.8714>,<at 0.75 0.7529>,<at 0.791 0.7529>,<at 0.9596 0.312>]}
        }}
       CurveIndex=Linked() {linkedVariable {name=NPC.FortniteNPC.TimeOfDay_0-1, type=NiagaraFloat}}
      ))
    )
    ScaleA=true
    ScaleRGB=true
   ) {module=ScaleColor001}
   &ScaleRibbonWidth(Ribbon Width Scale=&FloatFromCurve(FloatCurve=DataInterface() {propertyValues {Curve {keys [
         {arriveTangent=3.5755, interpMode=RCIM_Cubic, leaveTangent=3.5755, tangentMode=RCTM_User, time=0, value=1}
         {interpMode=RCIM_Cubic, leaveTangent=-0.1003, time=1, value=3}
        ]}}}))
   &FN_UpdateAnimTrail()
   /SetVariables(Particles.InvAge:NiagaraFloat=&OneMinusFloat(Float=$Particles.NormalizedAge)) {isSetParameters=true}
  ]
  rendererClasses [&NiagaraRibbonRendererProperties]
  renderers [{rendererClass=&NiagaraRibbonRendererProperties}]
 }
}
summary {
 systemName=NS_Pickaxe_ElegantHeist_Swing
 emitters [
  {emitterName=Trail_Details, rendererClasses [&NiagaraRibbonRendererProperties]}
  {emitterName=ChromaTrail, rendererClasses [&NiagaraRibbonRendererProperties]}
  {emitterName=Emissive_Filler_Trail, rendererClasses [&NiagaraRibbonRendererProperties]}
  {emitterName=Bling_Big, rendererClasses [&NiagaraSpriteRendererProperties, &NiagaraMeshRendererProperties]}
 ]
 userVariables [
  User.bIsFrontEnd {type=NiagaraBool}
  User.bIsFrontEndPreview {type=NiagaraBool}
  User.RibbonIDOffset {type=NiagaraInt32}
 ]
}
