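# Physics asset for the Elegant Heist two-handed pickaxe, providing bodies and joints for a short dangling chain on the weapon.
# The handle bone carries one large offset capsule, and base_chain plus three dyn_chain bones each carry a small capsule.
# Constraints link the chain in series back to the handle, with tight swing limits and zero twist limit between segments.
# The base_chain joint to the handle swings and twists freely.
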PhysicsAsset /BRCosmetics/Weapons/FORT_Melee/Pickaxe_ElegantHeist/Meshes/ElegantHeist_Axe_Physics
bodies {
 Melee_TwoHanded_Handle [{
   shapeType=Capsule
   length=15.4769
   radius=10.5993
   shapeName=dyn_mic_capsule
   center=<-2.6852 1.3423 79.684>
   rotation=<rot 30 -0.0002 -0.0004>
  }]
 base_chain [{shapeType=Capsule, length=0.505, radius=0.505, shapeName=base_chain_capsule, rotation=<rot 0 -90 0>}]
 dyn_chain_01 [{shapeType=Capsule, length=2.5, radius=1, shapeName=dyn_chain_01_capsule, center=<2.0548 0 0>, rotation=<rot 90 0 0>}]
 dyn_chain_02 [{shapeType=Capsule, length=2.5, radius=1, shapeName=dyn_chain_02_capsule, center=<2.445 0 0>, rotation=<rot 90 0 0>}]
 dyn_chain_03 [{shapeType=Capsule, length=2.5, radius=1, shapeName=dyn_chain_03_capsule, center=<2.7063 0 0>, rotation=<rot 90 0 0>}]
}
constraints [
 {bone1Name=dyn_chain_03, bone2Name=dyn_chain_02, swing1LimitDegrees=15, swing2LimitDegrees=15, twistLimitDegrees=0}
 {bone1Name=dyn_chain_02, bone2Name=dyn_chain_01, swing1LimitDegrees=15, swing2LimitDegrees=15, twistLimitDegrees=0}
 {bone1Name=dyn_chain_01, bone2Name=base_chain, swing1LimitDegrees=15, swing2LimitDegrees=15, twistLimitDegrees=0}
 {
  bone1Name=base_chain
  bone2Name=Melee_TwoHanded_Handle
  swing1LimitDegrees=45
  swing1Motion=Free
  swing2LimitDegrees=45
  swing2Motion=Free
  twistLimitDegrees=45
  twistMotion=Free
 }
]
